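; Quest ability for the Season 13 punch card that credits the player for distance driven with teammates in the vehicle.
; Tag waits start a repeating one-second timer on becoming driver and clear it on becoming passenger; a vehicle exit event ends the ability.
; While driving with more than one vehicle occupant, each tick sends straight-line meters since the last sample, floored, as a DrivePassengers quest stat.
; Ending the ability clears the timer and runs one final distance check.
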(event Ability|EventActivateAbilityFromEvent
  (params (EventData :type Gameplay%20Event%20Data%20Structure%20%28by%20ref%29))
  (|CallParentFunction EventData)
  ;[1 1 1] Becoming passenger
  (Ability|Tasks|WaitGameplayTagAdd "(TagName=\"MovementMode.Passenger\")"
    (:then
      ;[1 1 1] Becoming Driver
      (Ability|Tasks|WaitGameplayTagAdd "(TagName=\"MovementMode.Driving\")"
        (:then
          ;[1 1 1] Exiting vehicle
          (Ability|Tasks|WaitGameplayEvent "(TagName=\"Event.Vehicle.Exit\")"
            (:EventReceived
              (Ability|EndAbility))))
        (:Added
          (bind _returnvalue (Utilities|Time|SetTimerbyEvent (delegate Custom|TrackDistance) 1.0 true))
          (Variables|GAVars|Data|SetTimer_Progress _returnvalue)
          (Variables|GAVars|Data|SetPreviousLocation (Transformation|GetActorLocation (Variables|GAVars|Data|GetOwningPlayerPawn))))))
    (:Added
      (Utilities|Time|ClearandInvalidateTimerbyHandle (Variables|GAVars|Data|GetTimer_Progress)))))

(event Custom|TrackDistance
  ;[1 1 1] Track distance
  ;;[1 1 1] Make sure player is driving vehicle with passengers
  (bind _vehicle_occupied (Utilities|IsinVehicle (Variables|GAVars|Data|GetOwningPlayerPawn))
    (:true
      (if (and (> (Utilities|Array|Length (Vehicle|GetAllPassengers _vehicle_occupied)) 1) (Vehicle|IsDrivingVehicle (Variables|GAVars|Data|GetOwningPlayerPawn)))
        ;;[1 1 1] Calculate distance traveled
        (bind _output_get (Variables|GAVars|Data|SetCurrentLocation (Transformation|GetActorLocation (Variables|GAVars|Data|GetOwningPlayerPawn))))
        (Variables|GAVars|Data|SetMetersTraveled (/ (Math|Vector|VectorLength (- _output_get (Variables|GAVars|Data|GetPreviousLocation))) 1e+02))
        ;;[1 1 1] Update new location
        (Variables|GAVars|Data|SetPreviousLocation (Variables|GAVars|Data|GetCurrentLocation))
        (CallFunction|SendQuestProgress)))))

(event Custom|SendQuestProgress
  ;[1 1 1] Send quest progress
  (Quest|SendComplexCustomStatEvent (Variables|GAVars|Data|GetOwningPlayerQuestManager) 0 "(GameplayTags=((TagName=\"Athena.Quests.DrivePassengers\")))" "(GameplayTags=)" (Math|Float|Floor (Variables|GAVars|Data|GetMetersTraveled))))

(event Ability|EventOnEndAbility
  (params (bWasCancelled :type Boolean))
  ;[1 1 1] On end ability
  (|CallParentFunction bWasCancelled)
  (Utilities|Time|ClearandInvalidateTimerbyHandle (Variables|GAVars|Data|GetTimer_Progress))
  (CallFunction|TrackDistance))
